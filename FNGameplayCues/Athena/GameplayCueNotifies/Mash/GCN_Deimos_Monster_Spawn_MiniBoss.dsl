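; Burst gameplay cue for the Deimos monster mini-boss spawn in Mash, pinning its Cascade spawn effect to the target pawn.
; On burst, it casts the target to FortPawn and attaches the first particle component to the mesh's root socket, snapping location and rotation while keeping world scale.
; A failed cast does nothing, and OnExecute only forwards to the parent implementation.
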(fn OnBurst(Cascade)
  (params (MyTarget :type Actor%20Object%20Reference) (Parameters :type Gameplay%20Cue%20Parameters%20Structure%20%28by%20ref%29) (ParticleComponents :type Array%20of%20Cascade%20Particle%20System%20Component%20Object%20References) (AudioComponents :type Array%20of%20Audio%20Component%20Object%20References) (BurstCameraShakeInstance :type Camera%20Shake%20Base%20Object%20Reference) (BurstDecalInstance :type Decal%20Actor%20Object%20Reference))
  (bind _asfort_pawn (Utilities|Casting|CastToFortPawn MyTarget)
    (:then
      (Transformation|AttachComponentToComponent (Utilities|Array|Get(acopy) ParticleComponents) (Variables|/Script/Engine.Character|GetMesh _asfort_pawn) "root" "SnapToTarget" "SnapToTarget" "KeepWorld" true))
    (:CastFailed
      _))
  (return))

(fn OnExecute
  (params (MyTarget :type Actor%20Object%20Reference) (Parameters :type Gameplay%20Cue%20Parameters%20Structure%20%28by%20ref%29))
  (return (|CallParentFunction MyTarget Parameters)))
